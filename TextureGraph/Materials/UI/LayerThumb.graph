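# UI-domain translucent material that displays a layer thumbnail from the ThumbTex texture parameter, tinted by vertex color.
# SingleChannel blends the RGB toward the red channel replicated as grayscale, then the result is multiplied by vertex color.
# UseSourceOpacity blends an alpha value from fully opaque toward texture alpha times vertex alpha.
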Material /TextureGraph/Materials/UI/LayerThumb
graph {
 properties {blendMode=BLEND_Translucent, materialDomain=MD_UI, pixelDepthOffsetMode=PDOM_Legacy, usedShadingModels=Default Lit}
 nodes [
  n1 = TextureSampleParameter2D() {parameterName=ThumbTex, texture=/TextureGraph/Textures/default/default_Albedo}
  n2 = VertexColor()
  LinearInterpolate(A=Constant() {r=1}, B=Multiply(A=@n1.A, B=@n2.A), Alpha=ScalarParameter() {parameterName=UseSourceOpacity})
  Multiply(
   A=LinearInterpolate(
    A=@n1.RGB
    B=MaterialFunctionCall(R=@n1.R, G=@n1.R, B=@n1.R) {materialFunction=&AppendMany}.RGB
    Alpha=ScalarParameter() {parameterName=SingleChannel}
   )
   B=@n2
  )
 ]
}
